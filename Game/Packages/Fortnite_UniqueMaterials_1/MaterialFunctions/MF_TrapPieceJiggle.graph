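# Material function that adds a timestamp-driven jiggle offset to incoming material attributes, named for trap pieces.
# MF_Timestamp_InOut_01 turns two timestamps and transition lengths into a smooth step, which a contrast power shapes.
# That envelope feeds a biased sine, scaled by an amount along a transformed local-space vector and red vertex color, and also fades the result in.
# A static switch either adds this as World Position Offset or passes the material through unchanged.
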MaterialFunction /Game/Packages/Fortnite_UniqueMaterials_1/MaterialFunctions/MF_TrapPieceJiggle
graph {nodes [
  n1 = FunctionInput() {inputType=FunctionInput_MaterialAttributes}
  n2 = Power(
   Base=MaterialFunctionCall(
    TimestampA=ScalarParameter() {group=WPO_Jiggle, parameterName=TimestampA, sortPriority=3}
    TransitionLength_In=ScalarParameter() {group=WPO_Jiggle, parameterName=In_Trans_Length, sortPriority=4}
    TimestampB=ScalarParameter() {group=WPO_Jiggle, parameterName=TimestampB, sortPriority=5}
    TransitionLength_Out=ScalarParameter() {defaultValue=0.5, group=WPO_Jiggle, parameterName=Out_Trans_Length, sortPriority=6}
    Transition In Only?=StaticBool()
   ) {materialFunction=&MF_Timestamp_InOut_01}.SmoothStep
   Exp=ScalarParameter() {defaultValue=2.5, parameterName=Timestamp Contrast}
  )
  n3 = Multiply(A=Add(A=Sine(None=@n2) {period=0.5}) {constB=0.5}, B=ScalarParameter() {defaultValue=5, parameterName=JiggelAmount})
  n4 = Transform(None=VectorParameter() {parameterName=AmountVector, defaultValue=<rgba 0 1 0 0>}.RGB) {transformSourceType=TRANSFORMSOURCE_Local}
  n5 = LinearInterpolate(B=Multiply(A=Multiply(A=@n3, B=@n4), B=VertexColor().R), Alpha=@n2)
  FunctionOutput(None=StaticSwitchParameter(
    True=MaterialFunctionCall!(Material=@n1, WorldPositionOffset=@n5) {materialFunction=/Engine/Functions/MaterialLayerFunctions/MatLayerBlend_AddWorldPositionOffset}.Blended Material
    False=@n1
   ) {parameterName=Use Jiggle WPO})
 ]}
