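# Master opaque, lit material for Fortnite traps, with switchable normal, emissive and Athena team-color features.
# Diffuse border tint blends to a team color via isFriendly, while emissive combines optional texture cropping, a timed sweep along bounding-box UVW and team glow.
# These and LOD-effect plus black-hole world position offsets feed a deprecated base shader function.
# Its attributes, with packed specular masks, pass through Environment Effects, Creative move tool and trap-piece jiggle functions.
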Material /Game/Packages/Fortnite_UniqueMaterials_1/Traps/M_Trap_Master
graph {
 properties {
  bUseMaterialAttributes=true
  bUsedWithMeshDeformer=true
  bUsedWithMeshParticles=true
  bUsedWithNanite=true
  bUsedWithNiagaraMeshParticles=true
  bUsedWithSkeletalMesh=true
  pixelDepthOffsetMode=PDOM_Legacy
  usedShadingModels=Default Lit
 }
 nodes [
  n1 = TextureSampleParameter2D() {group=Diffuse, parameterName=Diffuse, texture=/Engine/EngineMaterials/DefaultDiffuse}
  n2 = TextureSampleParameter2D() {parameterName=DiffuseColorMask, texture=&T_BlacksRGB}
  n3 = VectorParameter() {group=Diffuse, parameterName=DiffuseBorderColor, defaultValue=<rgba 0.7157 0.4793 0.0296 1>}
  n4 = LinearInterpolate(A=@n1.RGB, B=Multiply(A=@n1.RGB, B=Multiply(A=@n2.RGB, B=@n3.RGB)), Alpha=@n2.G)
  n5 = TextureSampleParameter2D() {parameterName=DiffuseColorMask, texture=&T_BlacksRGB}
  n6 = VectorParameter() {group=AthenaTeamEmissive, parameterName=DiffuseTeamColor, defaultValue=<rgba 0.052 0.2936 0.7157 1>}
  n7 = LinearInterpolate(A=@n1.RGB, B=Multiply(A=@n1.RGB, B=Multiply(A=@n5.RGB, B=@n6.RGB)), Alpha=@n5.G)
  n8 = ScalarParameter() {group=AthenaTeamEmissive, parameterName=isFriendly}
  n9 = StaticComponentMaskParameter(None=MaterialFunctionCall() {materialFunction=&BoundingBoxBased_0-1_UVW}.RGB) {defaultB=true, group=Emissive Sweep, parameterName=Sweep Axis}
  n10 = Constant() {r=1}
  n11 = Time()
  n12 = ScalarParameter() {group=Emissive Sweep, parameterName=TimeStampOffset}
  n13 = Multiply(
   A=StaticSwitchParameter(True=Time() {bOverride_Period=true, period=1}, False=Add(A=@n11, B=@n12)) {group=Emissive Sweep, parameterName=Debug Sweep Time}
   B=ScalarParameter() {defaultValue=6, group=Emissive Sweep, parameterName=SweepSpeed}
  )
  n14 = OneMinus(None=@n13)
  n15 = Clamp(None=Add(
    A=DotProduct(A=@n9, B=@n10)
    B=StaticSwitchParameter(True=@n14, False=@n13) {defaultValue=true, group=Emissive Sweep, parameterName=Invert Sweep Direction}
   ) {constB=0.5})
  n16 = OneMinus(None=@n15)
  n17 = Power(
   Base=Multiply(A=Multiply(A=@n15, B=@n16)) {constB=4}
   Exp=ScalarParameter() {defaultValue=80, group=Emissive Sweep, parameterName=SweepExponent}
  )
  n18 = VectorParameter() {group=Emissive Sweep, parameterName=SweepColor, defaultValue=<rgba 0.26 1.3677 4 1>}
  n19 = LinearInterpolate(
   A=VectorParameter() {group=Emissive, parameterName=Blend_Color1, defaultValue=<rgba 0 0 0 1>}.RGB
   B=VectorParameter() {group=Emissive, parameterName=Blend_Color2, defaultValue=<rgba 0 0 0 1>}.RGB
   Alpha=ScalarParameter() {group=Emissive, parameterName=Blend}
  )
  n20 = VectorParameter() {group=Emissive, parameterName=Emissive_Crop_Mult, defaultValue=<rgba 0 0 0 1>}
  n21 = TextureObjectParameter() {group=Emissive, parameterName=Emissive_Texture_Cropped, texture=&DefaultDiffuse}
  n22 = VectorParameter() {group=Emissive, parameterName=Cropping_UpperL_Corner, defaultValue=<rgba 0 0 0 1>}
  n23 = VectorParameter() {group=Emissive, parameterName=Cropping_LowerR_Corner, defaultValue=<rgba 0 0 0 1>}
  n24 = StaticSwitchParameter(
   True=Multiply(
    A=StaticSwitchParameter(True=@n19, False=@n20.RGB) {group=Emissive, parameterName=Emissive_Cropping_Mult_Lerp}
    B=MaterialFunctionCall!(UVs, TextureIn=@n21, UpperLeftCorner=@n22.RGB, LowerRightCorner=@n23.RGB, ExportFloat4) {materialFunction=/Engine/Functions/Engine_MaterialFunctions01/Texturing/TextureCropping}.CroppedMasked
   )
   False=Multiply(
    A=TextureSampleParameter2D() {parameterName=Emissive, texture=&T_BlacksRGB}.RGB
    B=VectorParameter() {group=Emissive, parameterName=EmissiveMultiplier, defaultValue=<rgba 1 1 1 1>}.RGB
   )
  ) {group=Emissive, parameterName=Use_Emissive_Cropping}
  n25 = StaticSwitchParameter(True=Add(A=Multiply(A=@n17, B=@n18.RGB), B=@n24), False=@n24) {group=Emissive Sweep, parameterName=EmissiveSweep}
  n26 = Constant()
  n27 = TextureSampleParameter2D() {
   desc="UsesBChannel (Expected Normal map)"
   group=AthenaTeamEmissive
   parameterName=AthenaTeamEmissiveTexture
   samplerType=SAMPLERTYPE_Normal
   texture=/Game/Weapons/FORT_TRAPS/Floor_Player_Jump_Pad/Textures/T_JumpPads_N
  }
  n28 = Power(Base=@n27.B, Exp=ScalarParameter() {defaultValue=1, desc=PowerOnBChannelofNormal, group=AthenaTeamEmissive, parameterName=AthenaEmissivePow})
  n29 = TextureSampleParameter2D() {
   group=AthenaTeamEmissive
   parameterName=AthenaEmissiveTextureMask
   samplerType=SAMPLERTYPE_Masks
   texture=/Game/Weapons/FORT_TRAPS/Floor_MetalSpike_Trap/textures/T_MetalSpike_Trap_S
  }
  n30 = ChannelMaskParameter(None=@n29.RGB) {group=AthenaTeamEmissive, maskChannel=Green, parameterName=AthenaEmissiveTextureMaskChannel, defaultValue=<rgba 0 1 0 0>}
  n31 = LinearInterpolate(
   B=Multiply(
    A=Multiply(A=@n28, B=@n30.R)
    B=VectorParameter() {group=AthenaTeamEmissive, parameterName=AthenaTeamEmissiveMultiplier, defaultValue=<rgba 1 1 1 1>}.RGB
   )
   Alpha=ScalarParameter() {group=AthenaTeamEmissive, parameterName=isFriendly}
  )
  n32 = Constant()
  n33 = TextureSampleParameter2D() {
   group=Normal
   parameterName=Normals
   samplerType=SAMPLERTYPE_Normal
   texture=&T_JumpPads_N
  }
  n34 = VectorParameter() {parameterName=WorldPostionOfBlackHole, defaultValue=<rgba -30000 25536 1952 0>}
  n35 = AppendVector(A=@n34.RGB, B=@n34.RGB)
  n36 = MaterialFunctionCall!(
   World Position
   DistanceCullFadeAnimationOffset=MaterialFunctionCall() {materialFunction=/Game/Packages/Fortress_SharedMaterials/Base_Material_Functions/PickupEffectDistanceCullOverride}.Result
   Use actor center instead of pivot point
  ) {materialFunction=/Game/Packages/Fortress_SharedMaterials/Base_Material_Functions/LODEffect}
  n37 = Add(A=@n36.Result, B=WorldPosition().XYZ)
  n38 = MaterialFunctionCall!(
   -----------Base Shader-----------
   BaseColor1=StaticSwitchParameter(True=LinearInterpolate(A=@n4, B=@n7, Alpha=@n8), False=@n4) {group=AthenaTeamEmissive, parameterName=UseAthenaTeamEmissive}
   Base Color Alpha=@n1.A
   SpecularMasks1
   Emissive=Add(
    A=StaticSwitchParameter(True=@n25, False=@n26) {group=Emissive, parameterName=Emissive}
    B=StaticSwitchParameter(True=@n31, False=@n32) {parameterName=UseAthenaTeamEmissive}
   )
   Normals1=StaticSwitchParameter(True=@n33.RGB, False=Constant3Vector() {constant=<rgba 0 0 1 0>}) {group=Normal, parameterName=Use_Normal}
   World Position Offset=Add(
    A=MaterialFunctionCall!(WorldPostionOfBlackHole=@n35, World Position=@n37) {materialFunction=/Game/Packages/Fortnite_UniqueMaterials_1/BlackHoleFunction}.Result
    B=@n36.Result
   )
   DiffuseFresnelAmount=ScalarParameter() {defaultValue=0.5, group=Diffuse, parameterName=DiffuseFresnelAmount}
   -----------Destruction------------
   ------Multiple Diffuse Textures-----
   BaseColor2
   BaseColor3
   BaseColor4
   BaseColor5
   BaseColor6
   BaseColor7
   BaseColor8
   " - - - - - - - -"
   SpecularMasks2
   SpecularMasks3
   SpecularMasks4
   SpecularMasks5
   SpecularMasks6
   SpecularMasks7
   SpecularMasks8
   " - - - - - - - - "
   normal2
   Normal3
   Normal4
   Normal5
   Normal6
   Normal7
   Normal8
   " - - - - - - - - 1"
   Use 2 Textures
   Use 3 Textures
   Use 4 Textures
   Use 5 Textures
   Use 6 Textures
   Use 7 Textures
   Use 8 Textures
   Use 9 Textures
   Use 10 Textures
   ExposeLerpFunctionForInvertingTheHitDistanceCalc
  ) {materialFunction=/Game/Packages/Fortress_SharedMaterials/Base_Material_Functions/DEPRECATED_BaseShaderFunction}
  n39 = TextureSampleParameter2D() {
   group=Specular
   parameterName=SpecularMasks
   samplerType=SAMPLERTYPE_Masks
   texture=&T_Fortnite_Default_S
  }
  n40 = MakeMaterialAttributes!(
   BaseColor=@n38.Diffuse
   Metallic=Multiply(A=@n39.G, B=ScalarParameter() {defaultValue=1, group=Specular, parameterName=Metallic})
   Specular=@n39.R
   Roughness=LinearInterpolate(
    A=ScalarParameter() {defaultValue=0.1, group=Specular, parameterName=SpecRoughnessMin}
    B=ScalarParameter() {defaultValue=1, group=Specular, parameterName=SpecRoughnessMax}
    Alpha=@n39.B
   )
   Anisotropy
   EmissiveColor=@n38.Diffuse
   Opacity
   OpacityMask
   Normal=@n38.Tangent Normal
   Tangent
   WorldPositionOffset=@n38.Diffuse
   SubsurfaceColor
   ClearCoat
   ClearCoatRoughness
   AmbientOcclusion
   Refraction
   CustomizedUVs_0
   CustomizedUVs_1
   CustomizedUVs_2
   CustomizedUVs_3
   CustomizedUVs_4
   CustomizedUVs_5
   CustomizedUVs_6
   CustomizedUVs_7
   PixelDepthOffset
   ShadingModel
   Displacement
  )
  MaterialFunctionCall!(In=MaterialFunctionCall(MA In=MaterialFunctionCall!(Environment FX=@n40) {materialFunction=/Game/Packages/Fortress_SharedMaterials/Base_Material_Functions/Environment_Effects}.Result) {materialFunction=&MF_CreativeMoveTool}.MA Out) {materialFunction=/Game/Packages/Fortnite_UniqueMaterials_1/MaterialFunctions/MF_TrapPieceJiggle}
  StaticBool() {value=true}
 ]
}
